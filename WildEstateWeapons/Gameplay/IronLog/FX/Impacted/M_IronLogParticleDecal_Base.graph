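# Translucent unlit base material for IronLog impact particles that fakes decals on the surfaces behind them.
# Mask UVs are projected from world position behind translucency, with optional panning, SubUV and erosion.
# Particle color is modulated by scene AO and platform-switched sun shading from scene normals.
# Opacity combines the mask, distance and gradient fades, quality-gated depth fade, and Perlin-Worley noise driving curve-atlas alpha and a second dissolve.
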Material /WildEstateWeapons/Gameplay/IronLog/FX/Impacted/M_IronLogParticleDecal_Base
graph {
 properties {
  bUsedWithNiagaraMeshParticles=true
  bUsedWithNiagaraSprites=true
  blendMode=BLEND_Translucent
  pixelDepthOffsetMode=PDOM_Legacy
  shadingModel=MSM_Unlit
  usedShadingModels=Unlit
 }
 nodes [
  n1 = Time()
  n2 = ScalarParameter() {defaultValue=1, group=Panner, parameterName=Panner Speed Scale, sortPriority=1}
  n3 = ScalarParameter() {group=Panner, parameterName=Panner U Speed, sortPriority=2}
  n4 = ScalarParameter() {group=Panner, parameterName=Panner V Speed, sortPriority=3}
  n5 = MaterialFunctionCall() {materialFunction=&WorldPositionBehindTranslucency}
  n6 = ParticlePositionWS()
  n7 = ParticleSize()
  n8 = Transform(None=Divide(A=Subtract(A=@n5.Result, B=@n6), B=ComponentMask(None=@n7) {r=true})) {transformSourceType=TRANSFORMSOURCE_World, transformType=TRANSFORM_Tangent}
  n9 = ScalarParameter() {defaultValue=1, group=WorldScale, parameterName=World Scale U}
  n10 = ScalarParameter() {defaultValue=1, group=WorldScale, parameterName=World Scale V, sortPriority=1}
  n11 = Subtract(A=Multiply(A=ComponentMask(None=@n8) {g=true, r=true}, B=AppendVector(A=@n9, B=@n10)), B=Constant2Vector() {g=0.5, r=0.5})
  n12 = StaticSwitchParameter(True=Panner(Coordinate=@n11, Time=Multiply(A=@n1, B=@n2), Speed=AppendVector(A=@n3, B=@n4)), False=@n11) {group=Panner, parameterName=Use Panner}
  n13 = TextureCoordinate()
  n14 = ScalarParameter() {parameterName=Use normal Uvs for mask}
  n15 = TextureCoordinate()
  n16 = LinearInterpolate(
   A=TextureSampleParameter2D(UVs=LinearInterpolate(A=@n12, B=@n13, Alpha=@n14)) {parameterName=Mask Tex, texture=/Game/Effects/Fort_Effects/Materials/T_EFX_Flare_Burst_01}.RGBA
   B=TextureSampleParameterSubUV!(UVs=@n15, Apply View MipBias) {group=Textures, parameterName=SubUVTex, texture=/BRCosmetics/UI/Foundation/Textures/Banner/Bucket/T-Banners-Icons-Splat}.RGBA
   Alpha=ScalarParameter() {group=Panner, parameterName=UseSubUVTex, sortPriority=3}
  )
  AppendVector(A=ComponentMask(None=@n16) {b=true, g=true, r=true}, B=ComponentMask(None=@n16) {a=true})
  AppendVector()
  DynamicParameter() {defaultValue=<rgba 0 1 1 1>, paramNames [Erosion, cur offset, cur offset, cur offset]}
  n17 = Desaturation(None=ParticleColor().RGB, Fraction=ScalarParameter() {parameterName=desat})
  n18 = SceneTexture() {sceneTextureId=PPI_BaseColor}
  n19 = ComponentMask(None=SceneTexture() {sceneTextureId=PPI_AmbientOcclusion}.Color) {b=true, g=true, r=true}
  LinearInterpolate(A=@n17, B=ComponentMask(None=Multiply(A=@n18.Color, B=@n19)) {b=true, g=true, r=true}, Alpha=ScalarParameter() {parameterName=Under Color})
  n20 = Multiply(A=@n17, B=@n19)
  n21 = Multiply(A=@n20, B=ScalarParameter() {parameterName=shadow Mul}) {constB=0.5}
  n22 = Constant() {r=1}
  n23 = DotProduct(A=SkyAtmosphereLightDirection(), B=SceneTexture() {sceneTextureId=PPI_WorldNormal}.Color)
  n24 = VectorParameter() {parameterName=Shadow Mask Options, defaultValue=<rgba 1 1 0 0>}
  n25 = Multiply(A=Power(Base=ConstantBiasScale(None=@n23), Exp=@n24.R), B=@n24.G)
  n26 = Saturate(None=MaterialFunctionCall(In=Add(A=@n25, B=@n24.B), Contrast=@n24.A) {materialFunction=&CheapContrast}.Result)
  n27 = MaterialFunctionCall("Xb1,PS4"=@n22, "PC-High&&Epic,XSX,PS5"=@n26, "PC-Perf,Android"=@n22, Switch=@n22, "PC-Low(NotPerf)&&Medium"=@n26) {materialFunction=&PlatformSwitch_1D}
  MaterialFunctionCall(Emissive=Add(A=LinearInterpolate(A=@n21, B=@n20, Alpha=@n27.Result)) {constB=0}) {materialFunction=&MF_FX_EmissiveAdjustments}
  MaterialFunctionCall() {materialFunction=&DebugTimeSine}
  n28 = ReflectionVectorWS(CustomWorldNormal=ComponentMask(None=SceneTexture() {sceneTextureId=PPI_WorldNormal}.Color) {b=true, g=true, r=true})
  n29 = TextureSampleParameterCube!(UVs=@n28, Bias=ScalarParameter() {defaultValue=5, parameterName=mip bia}, Apply View MipBias) {
   mipValueMode=TMVM_MipBias
   parameterName=Cubemap
   samplerType=SAMPLERTYPE_LinearColor
   texture=/Game/Packages/Fortress_Sky/TexturesHDR/T_AthenaSkylight
  }
  n30 = MaterialFunctionCall!(Vector=@n28) {materialFunction=/Engine/Functions/Engine_MaterialFunctions01/Coordinates/LongLatToUV}
  n31 = VectorParameter() {parameterName=Reflection grad options, channelNames {a=contrast, b=add, g=mul, r=pow}, defaultValue=<rgba 1 1 0 2>}
  n32 = Power(Base=OneMinus(None=ComponentMask(None=@n30.Result) {g=true}), Exp=@n31.pow)
  n33 = MaterialFunctionCall(In=Add(A=Multiply(A=@n32, B=@n31.mul), B=@n31.add), Contrast=@n31.contrast) {materialFunction=&CheapContrast}
  n34 = ScalarParameter() {parameterName=Use refect textur}
  n35 = VectorParameter() {parameterName=Reflect texture tint / desat, defaultValue=<rgba 0.2075 0.0493 0.27 0>}
  Multiply(A=Desaturation(None=LinearInterpolate(A=@n29.RGB, B=@n33.Result, Alpha=@n34), Fraction=@n35.A), B=@n35.RGB)
  n36 = ParticlePositionWS()
  n37 = CameraPositionWS()
  n38 = ParticleSize()
  Multiply(A=Normalize(VectorInput=Subtract(A=@n36, B=@n37)), B=Multiply(A=ComponentMask(None=@n38) {r=true}) {constB=-0.5}) {desc=WPO input}
  NamedRerouteDeclaration(None=ScalarParameter() {parameterName=MipBias}) {name=Param MipBias, nodeColor=<rgba 0.875 0 1 1>}
  Reroute()
  Sign(None=@n23)
  n39 = Constant() {r=0.5}
  SphereMask(A=@n15, B=@n39, Radius=@n39, Hardness=ScalarParameter() {defaultValue=0.95, parameterName=Hardness})
  n40 = TextureCoordinate()
  n41 = Constant()
  n42 = DynamicParameter() {defaultValue=<rgba 0 1 1 1>, paramNames [Erosion, cur offset, cur offset, cur offset]}
  n43 = ScalarParameter() {defaultValue=1, parameterName=World Pos Density}
  n44 = Multiply(A=Add(A=AppendVector(A=@n40, B=@n41), B=ComponentMask(None=@n42.RGBA) {a=true, b=true, g=true}), B=@n43)
  n45 = TextureSample!(UVs=@n44, Tex, Bias=NamedRerouteUsage(), Apply View MipBias) {mipValueMode=TMVM_MipBias, samplerType=SAMPLERTYPE_LinearColor, texture=/Game/TimeOfDay/CloudsAndStorm/Textures/VT_PerlinWorley}
  n46 = TextureSample(UVs=Add(A=Divide(A=@n44) {constB=2.2652}, B=Multiply(A=@n43) {constB=1.885}), Bias=NamedRerouteUsage()) {mipValueMode=TMVM_MipBias, samplerType=SAMPLERTYPE_LinearColor, texture=&VT_PerlinWorley}
  n47 = TextureSample(UVs=Add(A=Divide(A=@n44) {constB=2.7559}, B=Multiply(A=@n43) {constB=3.3257}), Bias=NamedRerouteUsage()) {mipValueMode=TMVM_MipBias, samplerType=SAMPLERTYPE_LinearColor, texture=&VT_PerlinWorley}
  n48 = Multiply(A=Add(A=Add(A=@n45.R, B=@n46.G), B=@n47.B), B=ScalarParameter() {defaultValue=0.3, parameterName=CurveDencity}) {constB=10}
  n49 = ScalarParameter() {parameterName=Curve offset}
  n50 = CurveAtlasRowParameter(CurveTime=Frac(None=Add(A=@n48, B=@n49))) {
   atlas=/WildEstateWeapons/Gameplay/IronLog/FX/Decals/CA_IronLog
   curve=/WildEstateWeapons/Gameplay/IronLog/FX/Decals/CC_COlors
   parameterName=Color Curve
  }
  n51 = LinearInterpolate(
   A=@n16
   B=Subtract(A=@n16, B=DynamicParameter() {defaultValue=<rgba 0 1 1 1>, paramNames [Erosion, cur offset, cur offset, cur offset]}.Erosion) {constB=0.5}
   Alpha=ScalarParameter() {group=Panner, parameterName=UseErosion, sortPriority=3}
  )
  n52 = Divide(
   A=Distance(
    A=MaterialFunctionCall() {materialFunction=&WorldPositionBehindTranslucency}.Result
    B=ParticlePositionWS()
   )
   B=ScalarParameter() {defaultValue=5, group=Distance Fade, parameterName=Fade Distance}
  )
  n53 = ScalarParameter() {defaultValue=5, group=Distance Fade, parameterName=Fade Start Distance, sortPriority=1}
  n54 = OneMinus(None=Clamp(None=Subtract(A=@n52, B=@n53)))
  n55 = TextureCoordinate()
  n56 = ComponentMask(None=TextureCoordinate()) {r=true}
  n57 = Multiply(
   A=Power(
    Base=ComponentMask(None=@n55) {r=true}
    Exp=ScalarParameter() {defaultValue=1, group=Opacity, parameterName=Left-Right Contrast, sortPriority=2}
   )
   B=Power(Base=OneMinus(None=@n56), Exp=ScalarParameter() {defaultValue=1, group=Opacity, parameterName=Right-Left Contrast, sortPriority=3})
  )
  n58 = TextureCoordinate()
  n59 = ComponentMask(None=TextureCoordinate()) {g=true}
  n60 = Multiply(
   A=Power(
    Base=ComponentMask(None=@n58) {g=true}
    Exp=ScalarParameter() {defaultValue=1, group=Opacity, parameterName=Top-Bottom Contrast, sortPriority=4}
   )
   B=Power(Base=OneMinus(None=@n59), Exp=ScalarParameter() {defaultValue=1, group=Opacity, parameterName=Bottom-Top Contrast, sortPriority=5})
  )
  n61 = Clamp(None=Multiply(
    A=Multiply(A=@n57, B=@n60)
    B=ScalarParameter() {defaultValue=100, group=Opacity, parameterName=Opacity_Intensity_Compensator, sortPriority=1}
   ))
  n62 = Constant() {r=1}
  n63 = Multiply(
   A=Multiply(A=ParticleColor().A, B=StaticComponentMaskParameter(None=@n51) {defaultG=true, parameterName=Mask channel})
   B=Multiply(A=Multiply(A=@n54, B=@n54), B=StaticSwitchParameter(True=@n61, False=@n62) {group=Opacity, parameterName=Use Box Gradient Opacity Control})
  )
  n64 = DepthFade(Opacity=@n63, FadeDistance=ScalarParameter() {parameterName=DepthFadeDistance})
  n65 = QualitySwitch(Default=ShadingPathSwitch(Default=@n64, Mobile=@n63), Low=@n63, High=@n64, Medium=@n63, Epic=@n64)
  n66 = Add(
   A=@n45.G
   B=LinearInterpolate(Alpha=DynamicParameter() {parameterIndex=1, defaultValue=<rgba 0 1 1 1>, paramNames [Erosion 2, na, na, na]}.Erosion 2) {constA=-1}
  )
  n67 = ScalarParameter() {parameterName=COntrast 2}
  Subtract(
   A=Multiply(A=Multiply(A=@n50.A, B=@n65), B=VertexColor().A)
   B=Saturate(None=MaterialFunctionCall(In=@n66, Contrast=@n67) {materialFunction=&CheapContrast}.Result)
  )
  TextureSample!(UVs, Tex, Apply View MipBias) {texture=&T-Banners-Icons-Splat}
  Transform(None=MaterialFunctionCall(
    Normal=TextureSample() {samplerType=SAMPLERTYPE_Normal, texture=/Game/Effects/Fort_Effects/Gameplay/Floppers/FlopperUpdrade/T_PerlinNormal}.RGB
    Flatness=ScalarParameter() {defaultValue=0.5, parameterName=Normal Flatness}
   ) {materialFunction=&FlattenNormal}.Result)
 ]
}
