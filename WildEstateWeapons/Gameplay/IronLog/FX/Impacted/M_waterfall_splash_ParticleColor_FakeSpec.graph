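# Translucent unlit Niagara material for a toon water splash with fake specular, filed under IronLog impact effects.
# Emissive adds particle color to a tinted, desaturated reflection, blended between a normal-mapped skylight cubemap and a reflection-vector vertical gradient.
# Opacity erodes a height texture via dynamic parameter, rounds it for hard edges, then applies depth fade (skipped on mobile) and near-camera fade.
# Switchable world position offset follows water surface displacement.
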Material /WildEstateWeapons/Gameplay/IronLog/FX/Impacted/M_waterfall_splash_ParticleColor_FakeSpec
graph {
 properties {
  bTangentSpaceNormal=false
  bUsedWithNiagaraMeshParticles=true
  bUsedWithNiagaraSprites=true
  blendMode=BLEND_Translucent
  pixelDepthOffsetMode=PDOM_Legacy
  shadingModel=MSM_Unlit
  translucencyLightingMode=TLM_VolumetricPerVertexNonDirectional
  translucencyPass=MTP_BeforeDOF
  translucentSelfShadowDensityScale=0
  translucentSelfShadowSecondDensityScale=0
  translucentShadowDensityScale=0
  usedShadingModels=Unlit
 }
 nodes [
  n1 = TextureSample() {samplerType=SAMPLERTYPE_Normal, texture=/Game/Effects/Fort_Effects/Textures/Water/T_waterSplash_N}
  ComponentMask(None=@n1.RGB) {b=true, g=true, r=true}
  n2 = ParticleColor()
  n3 = ReflectionVectorWS(CustomWorldNormal=Transform(None=@n1.RGB))
  n4 = TextureSampleParameterCube!(UVs=@n3, Bias=ScalarParameter() {defaultValue=5, parameterName=mip bia}, Apply View MipBias) {
   mipValueMode=TMVM_MipBias
   parameterName=Cubemap
   samplerType=SAMPLERTYPE_LinearColor
   texture=/Game/Packages/Fortress_Sky/TexturesHDR/T_AthenaSkylight
  }
  n5 = OneMinus(None=ComponentMask(None=MaterialFunctionCall!(Vector=@n3) {materialFunction=/Engine/Functions/Engine_MaterialFunctions01/Coordinates/LongLatToUV}.Result) {g=true})
  n6 = VectorParameter() {parameterName=Reflection grad options, channelNames {a=contrast, b=add, g=mul, r=pow}, defaultValue=<rgba 1 1 0 2>}
  n7 = Add(A=Multiply(A=Power(Base=@n5, Exp=@n6.pow), B=@n6.mul), B=@n6.add)
  n8 = VectorParameter() {parameterName=Reflect texture tint / desat, defaultValue=<rgba 0.2075 0.0493 0.27 0>}
  n9 = Desaturation(
   None=LinearInterpolate(
    A=@n4.RGB
    B=MaterialFunctionCall(In=@n7, Contrast=@n6.contrast) {materialFunction=&CheapContrast}.Result
    Alpha=ScalarParameter() {parameterName=Use refect textur}
   )
   Fraction=@n8.A
  )
  MaterialFunctionCall(Emissive=Add(A=@n2.RGB, B=Multiply(A=@n9, B=@n8.RGB))) {materialFunction=&MF_FX_EmissiveAdjustments}
  n10 = TextureSample!(UVs, Tex, Bias=ScalarParameter() {defaultValue=2, parameterName=Bias}, Apply View MipBias) {
   mipValueMode=TMVM_MipBias
   samplerType=SAMPLERTYPE_LinearGrayscale
   texture=&/T_ToonWaterSpash_height
  }
  n11 = DynamicParameter() {paramNames [Erode,Param2,Param3,Param4]}
  n12 = Power(Base=@n10.R, Exp=LinearInterpolate(Alpha=@n11.Erode) {constA=5, constAlpha=0}) {constExponent=50}
  n13 = LinearInterpolate(Alpha=@n11.Erode) {constAlpha=0, constB=200}
  n14 = Multiply(A=@n10.R, B=Clamp(None=Multiply(A=@n12, B=@n13)))
  n15 = Multiply(A=@n2.A, B=Round(None=Add(A=@n14) {constB=0.45}))
  n16 = ScalarParameter() {defaultValue=100, parameterName=Depth Fade}
  Multiply(
   A=ShadingPathSwitch(Default=DepthFade(Opacity=@n15, FadeDistance=@n16), Mobile=@n15) {desc=Migrated from FeatureLevelSwitch}
   B=MaterialFunctionCall(Distance=ScalarParameter() {parameterName=CameraFade}) {materialFunction=&Near_Camera_Fade}.Result
  )
  SceneTexture() {sceneTextureId=PPI_WorldNormal}
  n17 = Multiply(
   A=MaterialFunctionCall!(Ignore XY Offsets=StaticBoolParameter() {defaultValue=true, parameterName=Ignore XY Gerstner Offsets}, Position, WaterBodyIndex, Depth) {materialFunction=/FortWater/Materials/Functions/WaterParticleDisplacement}.WPO
   B=VectorParameter() {parameterName=Water Displacement Multiplier, defaultValue=<rgba 1 1 1 0>}.RGB
  )
  n18 = MaterialFunctionCall(Z=ScalarParameter() {defaultValue=1, parameterName=ZOffset}) {materialFunction=&MakeFloat3}
  StaticSwitchParameter(
   True=LinearInterpolate(A=Constant(), B=Add(A=@n17, B=@n18.Result), Alpha=ScalarParameter() {defaultValue=1, parameterName=UseWaterDisplacement})
   False=Constant()
  ) {defaultValue=true, parameterName=UseWPO}
 ]
}
