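# Masked lit Niagara particle material for IronLog impact drops that swing like pendulums.
# Dynamic parameters supply current and previous positions, mesh origin, rest normal and rigidity for a UV-masked pivot rotation in WPO, using PreviousFrameSwitch for motion vectors.
# Static switches add optional noise-texture dissolve or swap the pendulum and noise offset for downward circular deformation.
# Particle color gets stepped sun-facing shading, and dithered noise drives pixel depth offset.
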Material /WildEstateWeapons/Gameplay/IronLog/FX/Impacted/Materials/M_Drop_PendulumWPO
graph {
 properties {
  allowTranslucentCustomDepthWrites=true
  bUsedWithNiagaraMeshParticles=true
  bUsedWithNiagaraSprites=true
  blendMode=BLEND_Masked
  pixelDepthOffsetMode=PDOM_Legacy
  usedShadingModels=Default Lit
 }
 nodes [
  Constant()
  Constant() {r=1}
  n1 = ParticleColor()
  n2 = Add(A=DotProduct(A=SkyAtmosphereLightDirection(), B=VertexNormalWS())) {constB=0.35}
  LinearInterpolate(A=Multiply(A=@n1.RGB) {constB=0.5}, B=@n1.RGB, Alpha=Round(None=Saturate(None=@n2)))
  n3 = Multiply(
   A=TextureCoordinate()
   B=ComponentMask(None=VectorParameter() {parameterName=UV Scale, defaultValue=<rgba 0.3 0.3 0 0>}.RGB) {g=true, r=true}
  )
  n4 = DynamicParameter() {paramNames [PrevPosX, PrevPosY, PrevPosZ, UV V Pan]}
  Multiply(A=@n3, B=@n4.UV V Pan)
  n5 = TextureCoordinate()
  n6 = ComponentMask(None=VectorParameter() {parameterName=PDO UV Scale, defaultValue=<rgba 2 2 0 0>}.RGB) {g=true, r=true}
  n7 = DynamicParameter() {parameterIndex=1, defaultValue=<rgba 0 0 -1 1>, paramNames [RestNormalX,RestNormalY,RestNormalZ,random]}
  n8 = TextureSample(UVs=Add(A=Multiply(A=@n5, B=@n6), B=@n7.random)) {texture=/Game/Effects/Fort_Effects/Textures/Tile/T_GaussianNoise_01_M}
  n9 = ScalarParameter() {defaultValue=1, parameterName=PDO pow}
  Multiply(
   A=MaterialFunctionCall(Alpha Threshold=Power(Base=@n8.R, Exp=@n9)) {materialFunction=&DitherTemporalAA}.Result
   B=ScalarParameter() {defaultValue=1, parameterName=PDO Scale}
  )
  n10 = TextureSampleParameter2D(UVs=Multiply(A=TextureCoordinate(), B=ScalarParameter() {defaultValue=1, parameterName=Dissolve Scale})) {parameterName=Dissolve Texture, samplerType=SAMPLERTYPE_LinearGrayscale, texture=/Game/Effects/Fort_Effects/Textures/Tile/T_CellNoise_Tile_M}
  n11 = OneMinus(None=@n10.R)
  n12 = Saturate(None=Add(A=StaticSwitchParameter(True=@n11, False=@n10.R) {parameterName=Invert Dissolve Tex}, B=@n1.A))
  n13 = Constant() {r=1}
  Round(None=Add(A=StaticSwitchParameter(True=@n12, False=@n13) {parameterName=Use Dissolve}, B=ScalarParameter() {parameterName=Opac Offset}))
  ScalarParameter() {parameterName=Met}
  ScalarParameter() {parameterName=spec}
  ScalarParameter() {parameterName=Rough}
  n14 = Constant3Vector() {constant=<rgba 0 0 -1 0>}
  n15 = TextureCoordinate()
  n16 = Constant() {r=0.5}
  n17 = Constant()
  n18 = OneMinus(None=Power(Base=SphereMask(A=@n15, B=@n16, Radius=@n16, Hardness=@n17) {hardnessPercent=0}) {constExponent=0.5})
  n19 = VectorParameter() {parameterName=WPO UV Scale, defaultValue=<rgba 2 2 0 0>}
  n20 = DynamicParameter() {parameterIndex=1, defaultValue=<rgba 0 0 -1 1>, paramNames [RestNormalX,RestNormalY,RestNormalZ,random]}
  n21 = Add(A=Multiply(A=TextureCoordinate(), B=ComponentMask(None=@n19.RGB) {g=true, r=true}), B=@n20.random)
  n22 = Multiply(
   A=Subtract(A=TextureSample(UVs=@n21) {samplerType=SAMPLERTYPE_LinearGrayscale, texture=&T_CellNoise_Tile_M}.RGB) {constB=0.5}
   B=VertexNormalWS()
  )
  n23 = ScalarParameter() {parameterName=WPO Noise Height}
  n24 = DynamicParameter() {parameterIndex=1, defaultValue=<rgba 0 0 -1 1>, paramNames [RestNormalX,RestNormalY,RestNormalZ,random]}
  n25 = Multiply(A=MaterialFunctionCall(R=@n24.RestNormalX, G=@n24.RestNormalX, B=@n24.RestNormalX) {materialFunction=&AppendMany}.RGB) {constB=-1}
  n26 = DynamicParameter() {parameterIndex=3, defaultValue=<rgba 1 1 1 0.15>, paramNames [PosX,PosY,PosZ,Rigidity]}
  n27 = MaterialFunctionCall(R=@n26.PosX, G=@n26.PosX, B=@n26.PosX) {materialFunction=&AppendMany}
  n28 = DynamicParameter() {parameterIndex=2, paramNames [MeshOriginX,MeshOriginY,MeshOriginZ,MaskFalloff]}
  n29 = MaterialFunctionCall(R=@n28.MeshOriginX, G=@n28.MeshOriginX, B=@n28.MeshOriginX, A=@n28.MeshOriginX) {materialFunction=&AppendMany}
  n30 = CrossProduct(A=@n25, B=Normalize(VectorInput=Subtract(A=@n27.RGB, B=@n29.RGB)))
  n31 = WorldPosition()
  n32 = Subtract(A=ComponentMask(None=TextureCoordinate()) {g=true}, B=ScalarParameter() {defaultValue=-0.2, parameterName=Mask Offset})
  n33 = Saturate(None=Power(Base=Saturate(None=@n32), Exp=@n28.MaskFalloff))
  n34 = DynamicParameter() {paramNames [PrevPosX, PrevPosY, PrevPosZ, UV V Pan]}
  n35 = MaterialFunctionCall(R=@n34.PrevPosX, G=@n34.PrevPosX, B=@n34.PrevPosX) {materialFunction=&AppendMany}
  n36 = CrossProduct(A=@n25, B=Normalize(VectorInput=Subtract(A=@n35.RGB, B=@n29.RGB)))
  n37 = PreviousFrameSwitch(
   Current Frame=LinearInterpolate(A=RotateAboutAxis(NormalizedRotationAxis=@n30, RotationAngle=@n26.Rigidity, PivotPoint=@n29.RGB, Position=@n31.XYZ), Alpha=@n33) {constB=0}
   Previous Frame=LinearInterpolate(A=RotateAboutAxis(NormalizedRotationAxis=@n36, RotationAngle=@n26.Rigidity, PivotPoint=@n29.RGB, Position=@n31.XYZ), Alpha=@n33) {constB=0}
  )
  n38 = Constant()
  StaticSwitchParameter(
   True=Multiply(
    A=Multiply(A=@n14, B=@n18)
    B=DynamicParameter() {parameterIndex=2, paramNames [MeshOriginX,MeshOriginY,MeshOriginZ,MaskFalloff]}.MaskFalloff
   )
   False=Add(A=Multiply(A=@n22, B=@n23), B=StaticSwitchParameter(True=@n37, False=@n38) {defaultValue=true, parameterName=IS_Pendulum})
  ) {parameterName=Use Circle Deforming}
  n39 = ParticleRandom()
  n40 = Multiply(A=ScalarParameter() {defaultValue=0.75, parameterName=UV Mul}, B=@n3)
  n41 = Add(A=@n3, B=@n4.UV V Pan)
  n42 = Constant2Vector() {g=1}
  TextureSample(UVs=Add(A=Add(A=@n39, B=@n40), B=Multiply(A=@n41, B=@n42))) {samplerType=SAMPLERTYPE_Masks, texture=/Game/Environments/Asteria/Terrain/Textures/T_Asteria_SmoothMud_S}
 ]
}
