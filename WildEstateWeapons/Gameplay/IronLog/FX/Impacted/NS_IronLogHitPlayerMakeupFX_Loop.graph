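# Niagara system for the IronLog weapon's makeup effect on a hit player, covering the body in colored sprites for the user-set effect length.
# Active chest and head-and-shoulders emitters burst sprites onto cylinder or sphere shapes that follow sampled skeletal mesh bones.
# A randomly offset curl-noise vector field, sampled on the first frame, drives patchy pink, yellow and blue coloring.
# Several decal and body-part emitters remain disabled.
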NiagaraSystem /WildEstateWeapons/Gameplay/IronLog/FX/Impacted/NS_IronLogHitPlayerMakeupFX_Loop
dependencies {
 usedDataInterfaces [
  &NiagaraDataInterfaceCurve
  &NiagaraDataInterfaceArrayColor
  &NiagaraDataInterfaceArrayFloat3
  /Script/Niagara.NiagaraDataInterfaceVectorField
  &NiagaraDataInterfaceArrayFloat
  &NiagaraDataInterfaceArrayFloat2
  &NiagaraDataInterfaceColorCurve
  &NiagaraDataInterfaceSkeletalMesh
 ]
 usedDynamicInputs [&Divide_Float, &Subtract_Float]
 usedModules [
  &self:SystemSpawnScript.SystemScriptSource.SystemScriptGraph.NiagaraNodeAssignment_0.SetVariables_#1
  &self:SystemSpawnScript.SystemScriptSource.SystemScriptGraph.NiagaraNodeAssignment_1.SetVariables_#2
  &SystemState
  &Timeline
  /Niagara/Modules/System/CompleteIfUnused
 ]
}
emitters {
 Body_Decals {
  bEnabled=false
  emitterUpdateScript [
   &EmitterState()
   &SpawnBurst_Instantaneous(Spawn Count=16, Use Spawn Probability=false, UseLoopCountLimit=true)
  ]
  particleSpawnScript [
   &InitializeParticle(
    Color=<rgba 0.5029 0.3712 0.2086 50>
    Lifetime=$User.EffectLength
    Sprite Rotation Mode=::NewEnumerator2
    Sprite Size Mode=::NewEnumerator3
    Uniform Sprite Size=55
    UsePositionOffset=false
    Write Material Random=true
   )
   &PartitionParticles(
    Normalized Index In Partition Includes One=true
    Particles Per Partition=16
    Partition Count Mode:ENiagara_InfiniteLoopDuration=::NewEnumerator1
    Partition Mode:ENiagara_PartitionMode_v2=::NewEnumerator1
   )
   &KillParticles(Kill Particles=&Boolean_OrOperation(
     Bool1=&SetBoolByIntComparison(A=$Particles.PartitionParticles.Index In Partition, B=3, Comparison Type=::NewEnumerator4)
     Bool2=&SetBoolByIntComparison(A=$Particles.PartitionParticles.Index In Partition, B=6, Comparison Type=::NewEnumerator4)
    ))
   &KillParticles(Kill Particles=&SetBoolByFloatComparison(A=&RandomRangeFloat(Minimum=0, Maximum=1), B=0.65)) {enabled=false, module=KillParticles001}
   &InitialMeshOrientation(
    Mesh Orientation Mode=::NewEnumerator6
    Orientation Quaternion:Quat4f=&MakeQuaternion(
     Quaternion From=0
     Axis=$User.ImpactNormal
     Angle=45
     From=?
     To=?
     Angle Type=::NewEnumerator0
     Coordinate Space=::World
     Normalize Vectors=false
    )
    Rotation=&RandomRangeVector(Minimum=<0 0 0>, Maximum=<0 0 0>)
    Use Rotation Vector=true
   )
   &SkeletalMeshLocation(
    Bone / Socket Index=$Particles.PartitionParticles.Index In Partition
    Bone Sampling Mode=::NewEnumerator4
    MaskSpawnGroup=false
    Orientation Sampling=::NewEnumerator1
    Skeletal Mesh=$System.Skeletal Mesh
   ) {module=SkeletalMeshLocation001}
   &KillParticles(Kill Particles=&SetBoolByFloatComparison(
     A=&DistanceBetweenPositions(
      A=$Particles.Position
      B=&TransformPosition(Position=$User.ImpactPosition, Source Space=::World, Destination Space=::Simulation)
     )
     B=35
    )) {module=KillParticles003}
  ]
  particleUpdateScript [
   &ParticleState(Kill Particles When Lifetime Has Elapsed=false, Let Infinitely Lived Particles Die When Emitter Deactivates=true)
   &SolveForcesAndVelocity(Clamp Velocity=false, Limit Acceleration=false)
   &DynamicMaterialParameters(
    Index 0 Param 1=&FloatFromCurve(
     FloatCurve=DataInterface() {propertyValues {Curve {
        defaultValue=1
        keys [<at 0.85 0 RCIM_Cubic>, {arriveTangent=9.2403, interpMode=RCIM_Cubic, leaveTangent=9.2403, tangentMode=RCTM_User, time=1, value=0.65}]
       }}}
     CurveIndex=$System.Timeline.NormalizedTime
     Scale Curve=0
    )
    Index 0 Param 2=&RandomRangeFloat(Minimum=0, Maximum=100)
    Index 0 Param 3=&RandomRangeFloat(Minimum=0, Maximum=100)
    Index 0 Param 4=&RandomRangeFloat(Minimum=0, Maximum=100)
    Index 1 Param 1=&FloatFromCurve(FloatCurve=DataInterface() {propertyValues {Curve {defaultValue=1, keys [<at 0.35 0 RCIM_Cubic>,<at 1 0.65 RCIM_Cubic>]}}})
    Module.Index0Float0Write=true
    Module.Index0Float1Write=true
    Module.Index0Float2Write=true
    Module.Index0Float3Write=true
    Module.Index1Float0Write=true
    Module.Index1Float1Write=true
    Module.Index1Float2Write=true
    Module.Index1Float3Write=true
    Write Parameter Index 1=::NewEnumerator1
   )
   &SkeletalMeshLocation(
    Bone / Socket Index=$Particles.PartitionParticles.Index In Partition
    Bone Sampling Mode=::NewEnumerator4
    MaskSpawnGroup=false
    Orientation Sampling=::NewEnumerator1
    Skeletal Mesh=$System.Skeletal Mesh
   )
   &AlignSpriteToMeshOrientation(
    Mesh Orientation Relative Sprite Alignment Vector=<1 0 0>
    Mesh Orientation Relative Sprite Facing Vector:Vector3f=<0 1 0>
    SetAlignment=true
    SetFacing=true
   ) {enabled=false}
   &ScaleColor(
    Scale Alpha=&FloatFromCurve(FloatCurve=DataInterface() {propertyValues {Curve {keys [<at 0 1 RCIM_Cubic>, {arriveTangent=-1.8165, interpMode=RCIM_Cubic, leaveTangent=-1.8165, tangentMode=RCTM_User, time=1, value=0.35}]}}})
    ScaleA=true
    ScaleRGB=true
   ) {enabled=false}
   &KillParticles(Kill Particles=&SetBoolByFloatComparison(A=$System.Timeline.NormalizedTime, B=0.999)) {module=KillParticles002}
   /SetVariables(
    Particles.Vector 2D:Vector2f=<50 50>
    Particles.Vector 4:Vector4f=&MakeVector4(
     X=&DistanceBetweenPositions(
      A=$Particles.Position
      B=&TransformPosition(Position=$User.ImpactPosition, Source Space=::World, Destination Space=::Simulation)
     )
     Y=0
     Z=0
     W=0
    )
   ) {isSetParameters=true, module=SetVariables_#3}
   /SetVariables() {isSetParameters=true, module=SetVariables_#4}
   &SpriteFacingAndAlignment(WriteAlignment=false, WriteFacing=true)
  ]
  rendererClasses [&NiagaraSpriteRendererProperties, &NiagaraMeshRendererProperties]
  renderers [
   {rendererClass=&NiagaraSpriteRendererProperties}
   {rendererClass=&NiagaraMeshRendererProperties, rendererIndex=1}
  ]
 }
 Body_Decals001 {
  bEnabled=false
  emitterUpdateScript [
   &EmitterState()
   &SpawnBurst_Instantaneous(Spawn Count=16, Use Spawn Probability=false, UseLoopCountLimit=true)
  ]
  particleSpawnScript [
   &InitializeParticle(
    Color=<rgba 0.5029 0.3712 0.2086 50>
    Lifetime=$User.EffectLength
    Sprite Rotation Mode=::NewEnumerator2
    Sprite Size Mode=::NewEnumerator3
    Uniform Sprite Size=55
    UsePositionOffset=false
    Write Material Random=true
   )
   &PartitionParticles(
    Normalized Index In Partition Includes One=true
    Particles Per Partition=16
    Partition Count Mode:ENiagara_InfiniteLoopDuration=::NewEnumerator1
    Partition Mode:ENiagara_PartitionMode_v2=::NewEnumerator1
   )
   &KillParticles(Kill Particles=&Boolean_OrOperation(
     Bool1=&SetBoolByIntComparison(A=$Particles.PartitionParticles.Index In Partition, B=3, Comparison Type=::NewEnumerator4)
     Bool2=&SetBoolByIntComparison(A=$Particles.PartitionParticles.Index In Partition, B=6, Comparison Type=::NewEnumerator4)
    ))
   &KillParticles(Kill Particles=&SetBoolByFloatComparison(A=&RandomRangeFloat(Minimum=0, Maximum=1), B=0.65)) {enabled=false, module=KillParticles001}
   &InitialMeshOrientation(
    Mesh Orientation Mode=::NewEnumerator6
    Orientation Quaternion:Quat4f=&MakeQuaternion(
     Quaternion From=0
     Axis=$User.ImpactNormal
     Angle=45
     From=?
     To=?
     Angle Type=::NewEnumerator0
     Coordinate Space=::World
     Normalize Vectors=false
    )
    Rotation=&RandomRangeVector(Minimum=<0 0 0>, Maximum=<0 0 0>)
    Use Rotation Vector=true
   )
   &SkeletalMeshLocation(
    Bone / Socket Index=$Particles.PartitionParticles.Index In Partition
    Bone Sampling Mode=::NewEnumerator4
    MaskSpawnGroup=false
    Orientation Sampling=::NewEnumerator1
    Skeletal Mesh=$System.Skeletal Mesh
   ) {module=SkeletalMeshLocation001}
   &KillParticles(Kill Particles=&SetBoolByFloatComparison(
     A=&DistanceBetweenPositions(
      A=$Particles.Position
      B=&TransformPosition(Position=$User.ImpactPosition, Source Space=::World, Destination Space=::Simulation)
     )
     B=35
    )) {module=KillParticles003}
   &CameraOffset(Camera Offset Amount=15)
  ]
  particleUpdateScript [
   &ParticleState(Kill Particles When Lifetime Has Elapsed=false, Let Infinitely Lived Particles Die When Emitter Deactivates=true)
   &SolveForcesAndVelocity(Clamp Velocity=false, Limit Acceleration=false)
   &DynamicMaterialParameters(
    Index 0 Param 1=&FloatFromCurve(
     FloatCurve=DataInterface() {propertyValues {Curve {
        defaultValue=1
        keys [<at 0.85 0 RCIM_Cubic>, {arriveTangent=9.2403, interpMode=RCIM_Cubic, leaveTangent=9.2403, tangentMode=RCTM_User, time=1, value=0.65}]
       }}}
     CurveIndex=$System.Timeline.NormalizedTime
     Scale Curve=0
    )
    Index 0 Param 2=&RandomRangeFloat(Minimum=0, Maximum=100)
    Index 0 Param 3=&RandomRangeFloat(Minimum=0, Maximum=100)
    Index 0 Param 4=&RandomRangeFloat(Minimum=0, Maximum=100)
    Index 1 Param 1=&FloatFromCurve(FloatCurve=DataInterface() {propertyValues {Curve {defaultValue=1, keys [<at 0.35 0 RCIM_Cubic>,<at 1 0.65 RCIM_Cubic>]}}})
    Module.Index0Float0Write=true
    Module.Index0Float1Write=true
    Module.Index0Float2Write=true
    Module.Index0Float3Write=true
    Module.Index1Float0Write=true
    Module.Index1Float1Write=true
    Module.Index1Float2Write=true
    Module.Index1Float3Write=true
    Write Parameter Index 1=::NewEnumerator1
   )
   &SkeletalMeshLocation(
    Bone / Socket Index=$Particles.PartitionParticles.Index In Partition
    Bone Sampling Mode=::NewEnumerator4
    MaskSpawnGroup=false
    Orientation Sampling=::NewEnumerator1
    Skeletal Mesh=$System.Skeletal Mesh
   )
   &AlignSpriteToMeshOrientation(
    Mesh Orientation Relative Sprite Alignment Vector=<1 0 0>
    Mesh Orientation Relative Sprite Facing Vector:Vector3f=<0 1 0>
    SetAlignment=true
    SetFacing=true
   ) {enabled=false}
   &ScaleColor(
    Scale Alpha=&FloatFromCurve(FloatCurve=DataInterface() {propertyValues {Curve {keys [<at 0 1 RCIM_Cubic>, {arriveTangent=-1.8165, interpMode=RCIM_Cubic, leaveTangent=-1.8165, tangentMode=RCTM_User, time=1, value=0.35}]}}})
    ScaleA=true
    ScaleRGB=true
   ) {enabled=false}
   &KillParticles(Kill Particles=&SetBoolByFloatComparison(A=$System.Timeline.NormalizedTime, B=0.999)) {module=KillParticles002}
   /SetVariables(
    Particles.Vector 2D:Vector2f=<50 50>
    Particles.Vector 4:Vector4f=&MakeVector4(
     X=&DistanceBetweenPositions(
      A=$Particles.Position
      B=&TransformPosition(Position=$User.ImpactPosition, Source Space=::World, Destination Space=::Simulation)
     )
     Y=0
     Z=0
     W=0
    )
   ) {isSetParameters=true, module=SetVariables_#3}
   /SetVariables() {enabled=false, isSetParameters=true, module=SetVariables_#4}
   &SpriteFacingAndAlignment(WriteAlignment=false, WriteFacing=false) {enabled=false}
  ]
  rendererClasses [&NiagaraSpriteRendererProperties]
  renderers [{rendererClass=&NiagaraSpriteRendererProperties}]
 }
 Chest {
  bEnabled=false
  emitterUpdateScript [
   &EmitterState()
   &SpawnBurst_Instantaneous(Spawn Count=500, Use Spawn Probability=false, UseLoopCountLimit=true)
  ]
  particleSpawnScript [
   &InitializeParticle(
    Color=&SelectLinearColorFromArray(
     Array Sampling Mode=::NewEnumerator0
     Direct Array Index=?
     Color Selection Array=DataInterface() {propertyValues {ColorData [<rgba 0.7835 0.1912 0.3916 1>, <rgba 1 1 0 1>, <rgba 0 0 1 1>]}}
     Random Seed=0
    )
    Lifetime=$User.EffectLength
    Sprite Rotation Mode=::NewEnumerator2
    Sprite Size Mode=::NewEnumerator3
    Uniform Sprite Size=25
    UsePositionOffset=false
    Write Material Random=true
   )
   &PartitionParticles(
    Normalized Index In Partition Includes One=true
    Particles Per Partition=16
    Partition Count Mode:ENiagara_InfiniteLoopDuration=::NewEnumerator1
    Partition Mode:ENiagara_PartitionMode_v2=::NewEnumerator1
   )
   &KillParticles(Kill Particles=&Boolean_OrOperation(
     Bool1=&SetBoolByIntComparison(A=$Particles.PartitionParticles.Index In Partition, B=3, Comparison Type=::NewEnumerator4)
     Bool2=&SetBoolByIntComparison(A=$Particles.PartitionParticles.Index In Partition, B=6, Comparison Type=::NewEnumerator4)
    ))
   &KillParticles(Kill Particles=&SetBoolByFloatComparison(A=&RandomRangeFloat(Minimum=0, Maximum=1), B=0.65)) {enabled=false, module=KillParticles001}
   &InitialMeshOrientation(
    Mesh Orientation Mode=::NewEnumerator6
    Orientation Quaternion:Quat4f=&MakeQuaternion(
     Quaternion From=0
     Axis=$User.ImpactNormal
     Angle=45
     From=?
     To=?
     Angle Type=::NewEnumerator0
     Coordinate Space=::World
     Normalize Vectors=false
    )
    Rotation=&RandomRangeVector(Minimum=<0 0 0>, Maximum=<0 0 0>)
    Use Rotation Vector=true
   )
   &SkeletalMeshLocation(
    Bone / Socket Index=$Particles.PartitionParticles.Index In Partition
    Bone Sampling Mode=::NewEnumerator4
    MaskSpawnGroup=false
    Orientation Sampling=::NewEnumerator1
    Skeletal Mesh=$System.Skeletal Mesh
   ) {module=SkeletalMeshLocation001}
   &KillParticles(Kill Particles=&SetBoolByFloatComparison(
     A=&DistanceBetweenPositions(
      A=$Particles.Position
      B=&TransformPosition(Position=$User.ImpactPosition, Source Space=::World, Destination Space=::Simulation)
     )
     B=35
    )) {enabled=false, module=KillParticles003}
   &CameraOffset(Camera Offset Amount=5)
   /SetVariables(Particles.Color=&SelectLinearColorFromArray(
     Array Sampling Mode=::NewEnumerator1
     Direct Array Index=&Add_Integer(
      A=&RandomRangeInt(Minimum=0, Maximum=2, Randomness Mode=::NewEnumerator1, Random Seed=$Emitter.RandomSeed, Fixed Random Seed=true)
      B=&RandomRangeInt(Minimum=0, Maximum=2, Randomness Mode=::NewEnumerator1, Random Seed=$Output.SkeletalMeshLocation001.SampledBoneIndex, Fixed Random Seed=true)
     )
     Direct Array Mode=::NewEnumerator1
     Color Selection Array=DataInterface() {propertyValues {ColorData [<rgba 0.7835 0.1912 0.3916 1>, <rgba 1 1 0 1>, <rgba 0 0 1 1>]}}
     Random Seed=?
    )) {isSetParameters=true, module=SetVariables_#5}
  ]
  particleUpdateScript [
   &ParticleState(Kill Particles When Lifetime Has Elapsed=false, Let Infinitely Lived Particles Die When Emitter Deactivates=true)
   &SolveForcesAndVelocity(Clamp Velocity=false, Limit Acceleration=false)
   &DynamicMaterialParameters(
    Index 0 Param 1=&FloatFromCurve(
     FloatCurve=DataInterface() {propertyValues {Curve {
        defaultValue=1
        keys [<at 0.85 0 RCIM_Cubic>, {arriveTangent=9.2403, interpMode=RCIM_Cubic, leaveTangent=9.2403, tangentMode=RCTM_User, time=1, value=0.65}]
       }}}
     CurveIndex=$System.Timeline.NormalizedTime
     Scale Curve=0
    )
    Index 0 Param 2=&RandomRangeFloat(Minimum=0, Maximum=100)
    Index 0 Param 3=&RandomRangeFloat(Minimum=0, Maximum=100)
    Index 0 Param 4=&RandomRangeFloat(Minimum=0, Maximum=100)
    Index 1 Param 1=&FloatFromCurve(FloatCurve=DataInterface() {propertyValues {Curve {defaultValue=1, keys [<at 0.35 0 RCIM_Cubic>,<at 1 0.65 RCIM_Cubic>]}}})
    Module.Index0Float0Write=true
    Module.Index0Float1Write=true
    Module.Index0Float2Write=true
    Module.Index0Float3Write=true
    Module.Index1Float0Write=true
    Module.Index1Float1Write=true
    Module.Index1Float2Write=true
    Module.Index1Float3Write=true
    Write Parameter Index 1=::NewEnumerator1
   )
   &SkeletalMeshLocation(
    Bone / Socket Index=$Particles.PartitionParticles.Index In Partition
    Bone Sampling Mode=::NewEnumerator4
    MaskSpawnGroup=false
    Orientation Sampling=::NewEnumerator1
    Sampled Position Offset=<500 0 0>
    Skeletal Mesh=$System.Skeletal Mesh
   )
   &ShapeLocation(
    Hemisphere Distribution=<1 0.7>
    MaskSpawnGroup=false
    Non Uniform Scale=&SelectVectorFromArray(
     Direct Array Index=$Particles.PartitionParticles.Index In Partition
     Vector Selection Array=DataInterface() {propertyValues {FloatData [<2 2.25 2>, <1 1.25 1.5>, <0 0 0>, <0 0 0>]}}
    )
    Offset=<0 1 4>
    Offset Mode=::NewEnumerator0
    Override Seed=false
    Shape Origin=$Particles.SkeletalMeshLocation.SampledPosition
    Sphere Radius=7
    Sphere Surface Distribution=1
    UseOwnerScale=true
   )
   &AlignSpriteToMeshOrientation(
    Mesh Orientation Relative Sprite Alignment Vector=<1 0 0>
    Mesh Orientation Relative Sprite Facing Vector:Vector3f=<0 1 0>
    SetAlignment=true
    SetFacing=true
   ) {enabled=false}
   &ScaleColor(
    Scale Alpha=&FloatFromCurve(FloatCurve=DataInterface() {propertyValues {Curve {keys [<at 0 1 RCIM_Cubic>, {arriveTangent=-1.8165, interpMode=RCIM_Cubic, leaveTangent=-1.8165, tangentMode=RCTM_User, time=1, value=0.35}]}}})
    ScaleA=true
    ScaleRGB=true
   ) {enabled=false}
   &KillParticles(Kill Particles=&SetBoolByFloatComparison(A=$System.Timeline.NormalizedTime, B=0.999)) {module=KillParticles002}
   /SetVariables(
    Particles.Vector 2D:Vector2f=<50 50>
    Particles.Vector 4:Vector4f=&MakeVector4(
     X=&DistanceBetweenPositions(
      A=$Particles.Position
      B=&TransformPosition(Position=$User.ImpactPosition, Source Space=::World, Destination Space=::Simulation)
     )
     Y=0
     Z=0
     W=0
    )
   ) {isSetParameters=true, module=SetVariables_#3}
   /SetVariables() {enabled=false, isSetParameters=true, module=SetVariables_#4}
   &SpriteFacingAndAlignment(Sprite Facing=$Particles.ShapeLocation.ShapeNormal, WriteAlignment=true, WriteFacing=true)
   /SetVariables(Particles.ALOKHBDFOIAWRBY:NiagaraInt32=$Output.SkeletalMeshLocation.SampledBoneIndex) {isSetParameters=true, module=SetVariables_#6}
  ]
  rendererClasses [&NiagaraSpriteRendererProperties]
  renderers [{rendererClass=&NiagaraSpriteRendererProperties}]
 }
 Chest001 {
  emitterSpawnScript [/SetVariables(Emitter.Vector:Vector3f=&RandomVector(Vector Scale=&RandomRangeFloat(Minimum=-10000000, Maximum=10000000))) {isSetParameters=true}]
  emitterUpdateScript [
   &EmitterState()
   &SpawnBurst_Instantaneous(Spawn Count=300, Spawn Group=1, Use Spawn Probability=false, UseLoopCountLimit=true)
  ]
  particleSpawnScript [
   &PartitionParticles(Number Of Partitions=5)
   &InitializeParticle(
    Color Mode=::NewEnumerator0
    Lifetime=$User.EffectLength
    Sprite Size Mode=::NewEnumerator3
    Uniform Sprite Size=12.5
    UsePositionOffset=false
    Write Material Random=true
   )
   &CameraOffset(Camera Offset Amount=3)
   /SetVariables(Particles.sort:NiagaraFloat=&RandomRangeFloat(Minimum=0, Maximum=100)) {isSetParameters=true}
  ]
  particleUpdateScript [
   &ParticleState()
   &SkeletalMeshLocation(
    Bone Sampling Mode=::NewEnumerator4
    MaskSpawnGroup=false
    Skeletal Mesh=$System.Skeletal Mesh
    Use Sampled Mesh Attribute as Ribbon ID=false
   )
   &ShapeLocation(
    Additional Quaternion Rotation=true
    Additional Yaw / Pitch / Roll=<0 0.25 -0.01>
    Cylinder Height=20
    Cylinder Height Midpoint=1
    Cylinder Radius=16
    MaskSpawnGroup=false
    Override Seed=false
    Rotation Mode=::NewEnumerator3
    Rotation Quaternion=$Particles.SkeletalMeshLocation.SampledRotation
    Shape Origin=$Particles.SkeletalMeshLocation.SampledPosition
    Shape Primitive=::NewEnumerator1
    Surface Only=true
    UseOwnerScale=true
   )
   &self:VecColor(
    Sample Point:Vector3f=&AddVectorToPosition(Position=$Particles.Position, Vector=$Emitter.Vector)
    Trigger Condition:NiagaraBool=$Output.ParticleState.FirstFrame
    Vector Field=DataInterface() {propertyValues {Field=/Niagara/VectorFields/Assets/TilingCurl16, bTileX=true, bTileY=true, bTileZ=true}}
    samplePoint Scale:Vector3f=&VectorFromFloat(Value=4.0039)
   ) {module=VecColor001}
   &Color(Color=&SelectLinearColorFromArray(
     Array Sampling Mode=::NewEnumerator1
     Direct Array Index=&MakeIntFromFloat(Float=&Multiply_Float(
       A=&FloatFromCurve(
        FloatCurve=DataInterface() {propertyValues {Curve {keys [{arriveTangent=-3.4034, time=0, value=1}, {leaveTangent=-3.6728, time=1, value=0}]}}}
        CurveIndex=&DotProduct(Vector A=$Particles.Sampled Value, Vector B=<1 0 0>)
       )
       B=4
      ))
     Direct Array Mode=::NewEnumerator1
     Color Selection Array=DataInterface() {propertyValues {ColorData [<rgba 1 0.1727 0.5535 1>, <rgba 1 0.8208 0 1>, <rgba 0 0.1387 1 1>]}}
     Random Seed=?
    )) {enabled=false, module=Color001}
   &SpriteFacingAndAlignment(
    Sprite Alignment=&ReturnMeshOrientationAxis(Mesh Orientation Axis=::XAxis, Mesh Orientation Quat=$Particles.MeshOrientation)
    Sprite Facing=&Lerp_Vector(A=$Particles.ShapeLocation.ShapeNormal, B=$Particles.ShapeLocation.ShapeVector, Alpha=0.05)
    WriteAlignment=true
    WriteFacing=true
   )
   &Color(
    Color=&ColorFromCurve(
     ColorCurve=DataInterface() {propertyValues {
       AlphaCurve {keys [
         {interpMode=RCIM_Constant, leaveTangentWeight=0.3333, time=0, value=1}
         {arriveTangentWeight=0.3333, interpMode=RCIM_Constant, time=1, value=1}
        ]}
       BlueCurve {keys [
         {interpMode=RCIM_Constant, leaveTangentWeight=0.0581, time=0, value=0.3514}
         <at 0.1466 0.3514>
         {arriveTangentWeight=0.0919, interpMode=RCIM_Constant, leaveTangentWeight=0.1, time=0.1814, value=0.3339}
         <at 0.4215 0.3339>
         <at 0.4864 0.4174 RCIM_Constant>
         <at 0.9122 0.4174>
         <at 0.9767 0.3514>
        ]}
       GreenCurve {keys [
         {interpMode=RCIM_Constant, leaveTangentWeight=0.0581, time=0, value=0.2284}
         <at 0.1466 0.2284>
         {arriveTangentWeight=0.0583, interpMode=RCIM_Constant, leaveTangentWeight=0.0634, time=0.1814, value=0.5166}
         <at 0.4215 0.5166>
         <at 0.4864 0.2237 RCIM_Constant>
         <at 0.9122 0.2237>
         <at 0.9767 0.2284>
        ]}
       RedCurve {keys [
         {interpMode=RCIM_Constant, leaveTangentWeight=0.0581, time=0, value=0.4957}
         <at 0.1466 0.4957>
         {arriveTangentWeight=0.1699, interpMode=RCIM_Constant, leaveTangentWeight=0.1848, time=0.1814, value=0.5565}
         <at 0.4215 0.5565>
         <at 0.4864 0.1778 RCIM_Constant>
         <at 0.9122 0.1778>
         <at 0.9767 0.4957>
        ]}
      }}
     CurveIndex=&FloatFromCurve(
      FloatCurve=DataInterface() {propertyValues {Curve {keys [
          {arriveTangent=-2.1269, interpMode=RCIM_Cubic, leaveTangent=-2.1269, tangentMode=RCTM_User, time=0, value=1}
          {interpMode=RCIM_Cubic, leaveTangent=-3.6728, time=1, value=0}
         ]}}}
      CurveIndex=&DotProduct(Vector A=$Particles.Sampled Value, Vector B=<1 0 0>)
     )
    )
    Scale Alpha=&FloatFromCurve(FloatCurve=DataInterface() {propertyValues {Curve {keys [<at 0 1 RCIM_Cubic>, {arriveTangent=-2.6709, interpMode=RCIM_Cubic, leaveTangent=-2.6709, tangentMode=RCTM_User, time=1, value=0}]}}})
   )
  ]
  rendererClasses [&NiagaraSpriteRendererProperties]
  renderers [
   {rendererClass=&NiagaraSpriteRendererProperties}
   {rendererClass=&NiagaraSpriteRendererProperties, rendererIndex=1}
  ]
 }
 Face_Decal {
  bEnabled=false
  emitterUpdateScript [
   &EmitterState()
   &SpawnBurst_Instantaneous(Use Spawn Probability=false, UseLoopCountLimit=true)
  ]
  particleSpawnScript [
   &InitializeParticle(
    Color=<rgba 0.5029 0.3712 0.2086 0.5>
    Lifetime=$User.EffectLength
    Sprite Rotation Angle=&RandomRangeFloat(Minimum=10, Maximum=-10)
    Sprite Rotation Mode=::NewEnumerator0
    Sprite Size Mode=::NewEnumerator3
    Uniform Sprite Size=15
    UsePositionOffset=false
    Write Material Random=true
   )
   &PartitionParticles(
    Normalized Index In Partition Includes One=true
    Particles Per Partition=16
    Partition Count Mode:ENiagara_InfiniteLoopDuration=::NewEnumerator1
    Partition Mode:ENiagara_PartitionMode_v2=::NewEnumerator1
   )
   &CameraOffset(Camera Offset Amount=7)
  ]
  particleUpdateScript [
   &SolveForcesAndVelocity(Clamp Velocity=false, Limit Acceleration=false)
   &AlignSpriteToMeshOrientation(
    Mesh Orientation Relative Sprite Alignment Vector=<1 0 0>
    Mesh Orientation Relative Sprite Facing Vector:Vector3f=<0.01 1 0>
    SetAlignment=true
    SetFacing=true
   )
   &ParticleState(Kill Particles When Lifetime Has Elapsed=false, Let Infinitely Lived Particles Die When Emitter Deactivates=true)
   &DynamicMaterialParameters(
    Index 0 Param 1=&FloatFromCurve(
     FloatCurve=DataInterface() {propertyValues {Curve {defaultValue=1, keys [<at 0.85 0 RCIM_Cubic>,<at 1 1 RCIM_Cubic>]}}}
     CurveIndex=$System.Timeline.NormalizedTime
     Scale Curve=0
    )
    Index 0 Param 2=&RandomRangeFloat(Minimum=0, Maximum=100)
    Index 0 Param 3=&RandomRangeFloat(Minimum=0, Maximum=100)
    Index 0 Param 4=&RandomRangeFloat(Minimum=0, Maximum=100)
    Index 1 Param 1=&FloatFromCurve(FloatCurve=DataInterface() {propertyValues {Curve {defaultValue=1, keys [<at 0.35 0 RCIM_Cubic>,<at 1 0.65 RCIM_Cubic>]}}})
    Module.Index0Float0Write=true
    Module.Index0Float1Write=true
    Module.Index0Float2Write=true
    Module.Index0Float3Write=true
    Module.Index1Float0Write=true
    Module.Index1Float1Write=false
    Module.Index1Float2Write=false
    Module.Index1Float3Write=false
    Write Parameter Index 1=::NewEnumerator1
   )
   &SkeletalMeshLocation(
    Bone / Socket Index=9
    Bone Sampling Mode=::NewEnumerator4
    MaskSpawnGroup=false
    Sampled Position Offset=<2 1 0>
    Skeletal Mesh=$System.Skeletal Mesh
   )
   &ScaleColor(
    Scale Alpha=&FloatFromCurve(FloatCurve=DataInterface() {propertyValues {Curve {keys [<at 0 1 RCIM_Cubic>, {arriveTangent=-1.8165, interpMode=RCIM_Cubic, leaveTangent=-1.8165, tangentMode=RCTM_User, time=1, value=0.35}]}}})
    ScaleA=true
    ScaleRGB=true
   ) {enabled=false}
   &KillParticles(Kill Particles=&SetBoolByFloatComparison(A=$System.Timeline.NormalizedTime, B=0.999)) {module=KillParticles002}
  ]
  rendererClasses [&NiagaraSpriteRendererProperties, &NiagaraMeshRendererProperties]
  renderers [
   {rendererClass=&NiagaraSpriteRendererProperties}
   {rendererClass=&NiagaraMeshRendererProperties, rendererIndex=1}
  ]
 }
 Head_n_shoulders001 {
  emitterSpawnScript [/SetVariables(Emitter.Vector:Vector3f=&RandomVector(Vector Scale=&RandomRangeFloat(Minimum=-10000000, Maximum=10000000))) {isSetParameters=true}]
  emitterUpdateScript [
   &EmitterState()
   &SpawnBurst_Instantaneous(Spawn Count=30, Spawn Group=3, Use Spawn Probability=false, UseLoopCountLimit=true) {module=SpawnBurst_Instantaneous002}
   &SpawnBurst_Instantaneous(Spawn Count=30, Spawn Group=2, Use Spawn Probability=false, UseLoopCountLimit=true) {module=SpawnBurst_Instantaneous001}
   &SpawnBurst_Instantaneous(Spawn Count=30, Spawn Group=1, Use Spawn Probability=false, UseLoopCountLimit=true)
  ]
  particleSpawnScript [
   &PartitionParticles(Number Of Partitions=3)
   &InitializeParticle(
    Color Mode=::NewEnumerator0
    Lifetime=$User.EffectLength
    Sprite Size Mode=::NewEnumerator3
    Uniform Sprite Size=12.5
    UsePositionOffset=false
    Write Material Random=true
   )
   &CameraOffset(Camera Offset Amount=5)
   /SetVariables(Particles.sort:NiagaraFloat=&RandomRangeFloat(Minimum=0, Maximum=100)) {isSetParameters=true}
  ]
  particleUpdateScript [
   &ParticleState()
   &SkeletalMeshLocation(
    Bone / Socket Index=&SpawnGroup()
    Bone Sampling Mode=::NewEnumerator4
    MaskSpawnGroup=false
    Sampled Position Offset=&SelectVectorFromArray(
     Direct Array Index=&SpawnGroup()
     Vector Selection Array=DataInterface() {propertyValues {FloatData [<1 2 3>, <5 1 0>, <2 0 0>, <-2 0 0>]}}
    )
    Skeletal Mesh=$System.Skeletal Mesh
   )
   &ShapeLocation(
    Additional Quaternion Rotation=true
    Additional Yaw / Pitch / Roll=&SelectVectorFromArray(
     Direct Array Index=&SpawnGroup()
     Evaluation Type=::NewEnumerator1
     Vector Selection Array=DataInterface() {propertyValues {FloatData [<0 0 0>, <0 -0.25 0>, <0 0 0>, <0 0.5 0>]}}
    )
    Hemisphere Distribution=&SelectVector2DFromArray(
     Direct Array Index=&SpawnGroup()
     Vector 2D Selection Array=DataInterface() {propertyValues {FloatData [<0 0>, <1 0.7>, <1 0.55>, <1 0.55>]}}
    )
    MaskSpawnGroup=false
    Non Uniform Scale=&SelectVectorFromArray(
     Direct Array Index=&SpawnGroup()
     Evaluation Type=::NewEnumerator1
     Vector Selection Array=DataInterface() {propertyValues {FloatData [<0 0 0>, <1 1.25 1.5>, <1.25 1 1>, <1.25 1 1>]}}
    )
    Override Seed=false
    Rotation Mode=::NewEnumerator3
    Rotation Quaternion=$Particles.SkeletalMeshLocation.SampledRotation
    Shape Origin=$Particles.Position
    Sphere Radius=&SelectFloatFromArray(
     Direct Array Index=&SpawnGroup()
     Float Selection Array=DataInterface() {propertyValues {FloatData [0,9,10,10]}}
    )
    Sphere Surface Distribution=1
    UseOwnerScale=true
   )
   &self:VecColor(
    Sample Point:Vector3f=&AddVectorToPosition(Position=$Particles.Position, Vector=$Emitter.Vector)
    Trigger Condition:NiagaraBool=$Output.ParticleState.FirstFrame
    Vector Field=DataInterface() {propertyValues {Field=&TilingCurl16, bTileX=true, bTileY=true, bTileZ=true}}
    samplePoint Scale:Vector3f=&VectorFromFloat(Value=4.0039)
   ) {module=VecColor}
   &Color(Color=&SelectLinearColorFromArray(
     Array Sampling Mode=::NewEnumerator1
     Direct Array Index=&MakeIntFromFloat(Float=&Multiply_Float(
       A=&FloatFromCurve(
        FloatCurve=DataInterface() {propertyValues {Curve {keys [{arriveTangent=-3.4034, time=0, value=1}, {interpMode=RCIM_Cubic, leaveTangent=-3.6728, time=1, value=0}]}}}
        CurveIndex=&DotProduct(Vector A=$Particles.Sampled Value, Vector B=<1 0 0>)
       )
       B=4
      ))
     Direct Array Mode=::NewEnumerator1
     Color Selection Array=DataInterface() {propertyValues {ColorData [<rgba 1 0.1727 0.5535 1>, <rgba 1 0.8208 0 1>, <rgba 0 0.1387 1 1>]}}
     Random Seed=?
    )) {enabled=false}
   &SpriteFacingAndAlignment(
    Sprite Alignment=&ReturnMeshOrientationAxis(Mesh Orientation Axis=::XAxis, Mesh Orientation Quat=$Particles.MeshOrientation)
    Sprite Facing=$Particles.ShapeLocation.ShapeVector
    WriteAlignment=true
    WriteFacing=true
   )
   &Color(Color=&ColorFromCurve(
     ColorCurve=DataInterface() {propertyValues {
       AlphaCurve {keys [
         {interpMode=RCIM_Constant, leaveTangentWeight=0.3333, time=0, value=1}
         {arriveTangentWeight=0.3333, interpMode=RCIM_Constant, time=1, value=1}
        ]}
       BlueCurve {keys [
         {interpMode=RCIM_Constant, leaveTangentWeight=0.0581, time=0, value=0.3514}
         <at 0.1467 0.3514>
         {arriveTangentWeight=0.0919, interpMode=RCIM_Constant, leaveTangentWeight=0.1, time=0.1814, value=0.3339}
         <at 0.4215 0.3339>
         <at 0.4864 0.4174 RCIM_Constant>
         <at 0.9122 0.4174>
         <at 0.9767 0.3514>
        ]}
       GreenCurve {keys [
         {interpMode=RCIM_Constant, leaveTangentWeight=0.0581, time=0, value=0.2284}
         <at 0.1467 0.2284>
         {arriveTangentWeight=0.0583, interpMode=RCIM_Constant, leaveTangentWeight=0.0634, time=0.1814, value=0.5166}
         <at 0.4215 0.5166>
         <at 0.4864 0.2237 RCIM_Constant>
         <at 0.9122 0.2237>
         <at 0.9767 0.2284>
        ]}
       RedCurve {keys [
         {interpMode=RCIM_Constant, leaveTangentWeight=0.0581, time=0, value=0.4957}
         <at 0.1467 0.4957>
         {arriveTangentWeight=0.1699, interpMode=RCIM_Constant, leaveTangentWeight=0.1848, time=0.1814, value=0.5565}
         <at 0.4215 0.5565>
         <at 0.4864 0.1778 RCIM_Constant>
         <at 0.9122 0.1778>
         <at 0.9767 0.4957>
        ]}
      }}
     CurveIndex=&FloatFromCurve(
      FloatCurve=DataInterface() {propertyValues {Curve {keys [
          {arriveTangent=-2.1269, interpMode=RCIM_Cubic, leaveTangent=-2.1269, tangentMode=RCTM_User, time=0, value=1}
          {interpMode=RCIM_Cubic, leaveTangent=-3.6728, time=1, value=0}
         ]}}}
      CurveIndex=&DotProduct(Vector A=$Particles.Sampled Value, Vector B=<1 0 0>)
      Scale Curve=&FloatFromCurve(FloatCurve=DataInterface() {propertyValues {Curve {keys [<at 0 1 RCIM_Cubic>, {arriveTangent=-3.6162, interpMode=RCIM_Cubic, leaveTangent=-3.6162, tangentMode=RCTM_User, time=1, value=0}]}}})
     )
    )) {module=Color001}
  ]
  rendererClasses [&NiagaraSpriteRendererProperties]
  renderers [{rendererClass=&NiagaraSpriteRendererProperties}]
 }
 head_ {
  bEnabled=false
  emitterUpdateScript [
   &EmitterState()
   &SpawnBurst_Instantaneous(Spawn Count=300, Use Spawn Probability=false, UseLoopCountLimit=true)
  ]
  particleSpawnScript [
   &InitializeParticle(
    Color=&SelectLinearColorFromArray(
     Array Sampling Mode=::NewEnumerator0
     Direct Array Index=?
     Color Selection Array=DataInterface() {propertyValues {ColorData [<rgba 0.7835 0.1912 0.3916 1>, <rgba 1 1 0 1>, <rgba 0 0 1 1>]}}
     Random Seed=0
    )
    Lifetime=$User.EffectLength
    Sprite Rotation Mode=::NewEnumerator2
    Sprite Size Mode=::NewEnumerator3
    Uniform Sprite Size=15
    UsePositionOffset=false
    Write Material Random=true
   )
   &PartitionParticles(
    Normalized Index In Partition Includes One=true
    Particles Per Partition=16
    Partition Count Mode:ENiagara_InfiniteLoopDuration=::NewEnumerator1
    Partition Mode:ENiagara_PartitionMode_v2=::NewEnumerator1
   )
   &KillParticles(Kill Particles=&Boolean_OrOperation(
     Bool1=&SetBoolByIntComparison(A=$Particles.PartitionParticles.Index In Partition, B=3, Comparison Type=::NewEnumerator4)
     Bool2=&SetBoolByIntComparison(A=$Particles.PartitionParticles.Index In Partition, B=6, Comparison Type=::NewEnumerator4)
    ))
   &KillParticles(Kill Particles=&SetBoolByFloatComparison(A=&RandomRangeFloat(Minimum=0, Maximum=1), B=0.65)) {enabled=false, module=KillParticles001}
   &InitialMeshOrientation(
    Mesh Orientation Mode=::NewEnumerator6
    Orientation Quaternion:Quat4f=&MakeQuaternion(
     Quaternion From=0
     Axis=$User.ImpactNormal
     Angle=45
     From=?
     To=?
     Angle Type=::NewEnumerator0
     Coordinate Space=::World
     Normalize Vectors=false
    )
    Rotation=&RandomRangeVector(Minimum=<0 0 0>, Maximum=<0 0 0>)
    Use Rotation Vector=true
   )
   &SkeletalMeshLocation(
    Bone / Socket Index=$Particles.PartitionParticles.Index In Partition
    Bone Sampling Mode=::NewEnumerator4
    MaskSpawnGroup=false
    Orientation Sampling=::NewEnumerator1
    Skeletal Mesh=$System.Skeletal Mesh
   ) {module=SkeletalMeshLocation001}
   &KillParticles(Kill Particles=&SetBoolByFloatComparison(
     A=&DistanceBetweenPositions(
      A=$Particles.Position
      B=&TransformPosition(Position=$User.ImpactPosition, Source Space=::World, Destination Space=::Simulation)
     )
     B=35
    )) {enabled=false, module=KillParticles003}
   &CameraOffset(Camera Offset Amount=5)
  ]
  particleUpdateScript [
   &ParticleState(Kill Particles When Lifetime Has Elapsed=false, Let Infinitely Lived Particles Die When Emitter Deactivates=true)
   &SolveForcesAndVelocity(Clamp Velocity=false, Limit Acceleration=false)
   &DynamicMaterialParameters(
    Index 0 Param 1=&FloatFromCurve(
     FloatCurve=DataInterface() {propertyValues {Curve {
        defaultValue=1
        keys [<at 0.85 0 RCIM_Cubic>, {arriveTangent=9.2403, interpMode=RCIM_Cubic, leaveTangent=9.2403, tangentMode=RCTM_User, time=1, value=0.65}]
       }}}
     CurveIndex=$System.Timeline.NormalizedTime
     Scale Curve=0
    )
    Index 0 Param 2=&RandomRangeFloat(Minimum=0, Maximum=100)
    Index 0 Param 3=&RandomRangeFloat(Minimum=0, Maximum=100)
    Index 0 Param 4=&RandomRangeFloat(Minimum=0, Maximum=100)
    Index 1 Param 1=&FloatFromCurve(FloatCurve=DataInterface() {propertyValues {Curve {defaultValue=1, keys [<at 0.35 0 RCIM_Cubic>,<at 1 0.65 RCIM_Cubic>]}}})
    Module.Index0Float0Write=true
    Module.Index0Float1Write=true
    Module.Index0Float2Write=true
    Module.Index0Float3Write=true
    Module.Index1Float0Write=true
    Module.Index1Float1Write=true
    Module.Index1Float2Write=true
    Module.Index1Float3Write=true
    Write Parameter Index 1=::NewEnumerator1
   )
   &SkeletalMeshLocation(
    Bone / Socket Index=9
    Bone Sampling Mode=::NewEnumerator4
    MaskSpawnGroup=false
    Orientation Sampling=::NewEnumerator1
    Sampled Position Offset=<500 0 0>
    Skeletal Mesh=$System.Skeletal Mesh
   )
   &ShapeLocation(
    MaskSpawnGroup=false
    Non Uniform Scale=<1 1.25 1.5>
    Offset=<0 1 4>
    Offset Mode=::NewEnumerator0
    Override Seed=false
    Shape Origin=$Particles.SkeletalMeshLocation.SampledPosition
    Sphere Radius=7
    Sphere Surface Distribution=1
    UseOwnerScale=true
   )
   &AlignSpriteToMeshOrientation(
    Mesh Orientation Relative Sprite Alignment Vector=<1 0 0>
    Mesh Orientation Relative Sprite Facing Vector:Vector3f=<0 1 0>
    SetAlignment=true
    SetFacing=true
   ) {enabled=false}
   &ScaleColor(
    Scale Alpha=&FloatFromCurve(FloatCurve=DataInterface() {propertyValues {Curve {keys [<at 0 1 RCIM_Cubic>, {arriveTangent=-1.8165, interpMode=RCIM_Cubic, leaveTangent=-1.8165, tangentMode=RCTM_User, time=1, value=0.35}]}}})
    ScaleA=true
    ScaleRGB=true
   ) {enabled=false}
   &KillParticles(Kill Particles=&SetBoolByFloatComparison(A=$System.Timeline.NormalizedTime, B=0.999)) {module=KillParticles002}
   /SetVariables(
    Particles.Vector 2D:Vector2f=<50 50>
    Particles.Vector 4:Vector4f=&MakeVector4(
     X=&DistanceBetweenPositions(
      A=$Particles.Position
      B=&TransformPosition(Position=$User.ImpactPosition, Source Space=::World, Destination Space=::Simulation)
     )
     Y=0
     Z=0
     W=0
    )
   ) {isSetParameters=true, module=SetVariables_#3}
   /SetVariables() {enabled=false, isSetParameters=true, module=SetVariables_#4}
   &SpriteFacingAndAlignment(Sprite Facing=$Particles.ShapeLocation.ShapeNormal, WriteAlignment=true, WriteFacing=true)
  ]
  rendererClasses [&NiagaraSpriteRendererProperties]
  renderers [{rendererClass=&NiagaraSpriteRendererProperties}]
 }
 head_n_Shoulders {
  bEnabled=false
  emitterUpdateScript [
   &EmitterState()
   &SpawnBurst_Instantaneous(Spawn Count=500, Use Spawn Probability=false, UseLoopCountLimit=true)
  ]
  particleSpawnScript [
   &InitializeParticle(
    Color=&SelectLinearColorFromArray(
     Array Sampling Mode=::NewEnumerator0
     Direct Array Index=?
     Color Selection Array=DataInterface() {propertyValues {ColorData [<rgba 0.7835 0.1912 0.3916 1>, <rgba 1 1 0 1>, <rgba 0 0 1 1>]}}
     Random Seed=0
    )
    Lifetime=$User.EffectLength
    Sprite Rotation Mode=::NewEnumerator2
    Sprite Size Mode=::NewEnumerator3
    Uniform Sprite Size=25
    UsePositionOffset=false
    Write Material Random=true
   )
   &PartitionParticles(
    Normalized Index In Partition Includes One=true
    Particles Per Partition=16
    Partition Count Mode:ENiagara_InfiniteLoopDuration=::NewEnumerator1
    Partition Mode:ENiagara_PartitionMode_v2=::NewEnumerator1
   )
   &KillParticles(Kill Particles=&Boolean_OrOperation(
     Bool1=&SetBoolByIntComparison(A=$Particles.PartitionParticles.Index In Partition, B=3, Comparison Type=::NewEnumerator4)
     Bool2=&SetBoolByIntComparison(A=$Particles.PartitionParticles.Index In Partition, B=6, Comparison Type=::NewEnumerator4)
    ))
   &KillParticles(Kill Particles=&SetBoolByFloatComparison(A=&RandomRangeFloat(Minimum=0, Maximum=1), B=0.65)) {enabled=false, module=KillParticles001}
   &InitialMeshOrientation(
    Mesh Orientation Mode=::NewEnumerator6
    Orientation Quaternion:Quat4f=&MakeQuaternion(
     Quaternion From=0
     Axis=$User.ImpactNormal
     Angle=45
     From=?
     To=?
     Angle Type=::NewEnumerator0
     Coordinate Space=::World
     Normalize Vectors=false
    )
    Rotation=&RandomRangeVector(Minimum=<0 0 0>, Maximum=<0 0 0>)
    Use Rotation Vector=true
   )
   &SkeletalMeshLocation(
    Bone / Socket Index=$Particles.PartitionParticles.Index In Partition
    Bone Sampling Mode=::NewEnumerator4
    MaskSpawnGroup=false
    Orientation Sampling=::NewEnumerator1
    Skeletal Mesh=$System.Skeletal Mesh
   ) {module=SkeletalMeshLocation001}
   &KillParticles(Kill Particles=&SetBoolByFloatComparison(
     A=&DistanceBetweenPositions(
      A=$Particles.Position
      B=&TransformPosition(Position=$User.ImpactPosition, Source Space=::World, Destination Space=::Simulation)
     )
     B=35
    )) {enabled=false, module=KillParticles003}
   &CameraOffset(Camera Offset Amount=5)
   /SetVariables(Particles.Color=&SelectLinearColorFromArray(
     Array Sampling Mode=::NewEnumerator1
     Direct Array Index=&Add_Integer(
      A=&RandomRangeInt(Minimum=0, Maximum=2, Randomness Mode=::NewEnumerator1, Random Seed=$Emitter.RandomSeed, Fixed Random Seed=true)
      B=&RandomRangeInt(Minimum=0, Maximum=2, Randomness Mode=::NewEnumerator1, Random Seed=$Output.SkeletalMeshLocation001.SampledBoneIndex, Fixed Random Seed=true)
     )
     Direct Array Mode=::NewEnumerator1
     Color Selection Array=DataInterface() {propertyValues {ColorData [<rgba 0.7835 0.1912 0.3916 1>, <rgba 1 1 0 1>, <rgba 0 0 1 1>]}}
     Random Seed=?
    )) {isSetParameters=true, module=SetVariables_#5}
  ]
  particleUpdateScript [
   &ParticleState(Kill Particles When Lifetime Has Elapsed=false, Let Infinitely Lived Particles Die When Emitter Deactivates=true)
   &SolveForcesAndVelocity(Clamp Velocity=false, Limit Acceleration=false)
   &DynamicMaterialParameters(
    Index 0 Param 1=&FloatFromCurve(
     FloatCurve=DataInterface() {propertyValues {Curve {
        defaultValue=1
        keys [<at 0.85 0 RCIM_Cubic>, {arriveTangent=9.2403, interpMode=RCIM_Cubic, leaveTangent=9.2403, tangentMode=RCTM_User, time=1, value=0.65}]
       }}}
     CurveIndex=$System.Timeline.NormalizedTime
     Scale Curve=0
    )
    Index 0 Param 2=&RandomRangeFloat(Minimum=0, Maximum=100)
    Index 0 Param 3=&RandomRangeFloat(Minimum=0, Maximum=100)
    Index 0 Param 4=&RandomRangeFloat(Minimum=0, Maximum=100)
    Index 1 Param 1=&FloatFromCurve(FloatCurve=DataInterface() {propertyValues {Curve {defaultValue=1, keys [<at 0.35 0 RCIM_Cubic>,<at 1 0.65 RCIM_Cubic>]}}})
    Module.Index0Float0Write=true
    Module.Index0Float1Write=true
    Module.Index0Float2Write=true
    Module.Index0Float3Write=true
    Module.Index1Float0Write=true
    Module.Index1Float1Write=true
    Module.Index1Float2Write=true
    Module.Index1Float3Write=true
    Write Parameter Index 1=::NewEnumerator1
   )
   &SkeletalMeshLocation(
    Bone / Socket Index=$Particles.PartitionParticles.Index In Partition
    Bone Sampling Mode=::NewEnumerator4
    MaskSpawnGroup=false
    Orientation Sampling=::NewEnumerator1
    Sampled Position Offset=<500 0 0>
    Skeletal Mesh=$System.Skeletal Mesh
   )
   &ShapeLocation(
    Hemisphere Distribution=<1 0.7>
    MaskSpawnGroup=false
    Non Uniform Scale=&SelectVectorFromArray(
     Direct Array Index=$Particles.PartitionParticles.Index In Partition
     Vector Selection Array=DataInterface() {propertyValues {FloatData [<2 2.25 2>, <1 1.25 1.5>, <2 2 2>, <2 2 2>]}}
    )
    Offset=<0 1 4>
    Offset Mode=::NewEnumerator0
    Override Seed=false
    Shape Origin=$Particles.SkeletalMeshLocation.SampledPosition
    Sphere Radius=7
    Sphere Surface Distribution=1
    UseOwnerScale=true
   )
   &AlignSpriteToMeshOrientation(
    Mesh Orientation Relative Sprite Alignment Vector=<1 0 0>
    Mesh Orientation Relative Sprite Facing Vector:Vector3f=<0 1 0>
    SetAlignment=true
    SetFacing=true
   ) {enabled=false}
   &ScaleColor(
    Scale Alpha=&FloatFromCurve(FloatCurve=DataInterface() {propertyValues {Curve {keys [<at 0 1 RCIM_Cubic>, {arriveTangent=-1.8165, interpMode=RCIM_Cubic, leaveTangent=-1.8165, tangentMode=RCTM_User, time=1, value=0.35}]}}})
    ScaleA=true
    ScaleRGB=true
   ) {enabled=false}
   &KillParticles(Kill Particles=&SetBoolByFloatComparison(A=$System.Timeline.NormalizedTime, B=0.999)) {module=KillParticles002}
   /SetVariables(
    Particles.Vector 2D:Vector2f=<50 50>
    Particles.Vector 4:Vector4f=&MakeVector4(
     X=&DistanceBetweenPositions(
      A=$Particles.Position
      B=&TransformPosition(Position=$User.ImpactPosition, Source Space=::World, Destination Space=::Simulation)
     )
     Y=0
     Z=0
     W=0
    )
   ) {isSetParameters=true, module=SetVariables_#3}
   /SetVariables() {enabled=false, isSetParameters=true, module=SetVariables_#4}
   &SpriteFacingAndAlignment(Sprite Facing=$Particles.ShapeLocation.ShapeNormal, WriteAlignment=true, WriteFacing=true)
   /SetVariables(Particles.float:NiagaraFloat=&self:ScratchDynamicInput_01(
     Vector Field:NiagaraDataInterfaceVectorField=DataInterface() {propertyValues {Field=/Niagara/VectorFields/Assets/TilingCurl32, bTileX=true, bTileY=true, bTileZ=true}}
     tile:NiagaraFloat=-1.2376
    )) {isSetParameters=true}
   /SetVariables(Particles.ALOKHBDFOIAWRBY:NiagaraInt32=$Output.SkeletalMeshLocation.SampledBoneIndex) {isSetParameters=true, module=SetVariables_#6}
   &Color(Color=&SelectLinearColorFromArray(
     Array Sampling Mode=::NewEnumerator1
     Direct Array Index=&MakeIntFromFloat(Float=&Multiply_Float(A=3, B=$Particles.float))
     Color Selection Array=DataInterface() {propertyValues {ColorData [<rgba 0.7835 0.1912 0.3916 1>, <rgba 1 1 0 1>, <rgba 0 0 1 1>]}}
     Random Seed=?
    ))
  ]
  rendererClasses [&NiagaraSpriteRendererProperties]
  renderers [{rendererClass=&NiagaraSpriteRendererProperties}]
 }
}
summary {
 systemName=NS_IronLogHitPlayerMakeupFX_Loop
 emitters [
  {
   bEnabled=false
   emitterName=Body_Decals
   rendererClasses [&NiagaraSpriteRendererProperties, &NiagaraMeshRendererProperties]
  }
  {
   bEnabled=false
   emitterName=Face_Decal
   rendererClasses [&NiagaraSpriteRendererProperties, &NiagaraMeshRendererProperties]
  }
  {bEnabled=false, emitterName=Body_Decals001, rendererClasses [&NiagaraSpriteRendererProperties]}
  {bEnabled=false, emitterName=head_, rendererClasses [&NiagaraSpriteRendererProperties]}
  {bEnabled=false, emitterName=head_n_Shoulders, rendererClasses [&NiagaraSpriteRendererProperties]}
  {bEnabled=false, emitterName=Chest, rendererClasses [&NiagaraSpriteRendererProperties]}
  {emitterName=Head_n_shoulders001, rendererClasses [&NiagaraSpriteRendererProperties]}
  {emitterName=Chest001, rendererClasses [&NiagaraSpriteRendererProperties]}
 ]
 userVariables [
  User.EffectLength {type=NiagaraFloat, defaultValue=4, description=""}
  User.ImpactNormal {type=Vector3f, description="", defaultValue {struct=Vector3f, value=<0 0 0>}}
  User.ImpactPosition {type=NiagaraPosition, description="", defaultValue {}}
  User.ImpactTraceStart {type=NiagaraPosition, description="", defaultValue {}}
 ]
}
